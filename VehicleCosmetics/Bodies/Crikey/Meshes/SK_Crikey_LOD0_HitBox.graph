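# Physics asset that defines the hitbox for the LOD0 skeletal mesh of the Crikey vehicle body cosmetic.
# A single box shape on the root bone, offset forward along X and raised along Z, stands in for the whole body.
# It has no other bodies and no constraints, so the collision shape has no articulated parts.
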PhysicsAsset /VehicleCosmetics/Bodies/Crikey/Meshes/SK_Crikey_LOD0_HitBox
bodies {root [{shapeType=Box, extentX=420.7755, extentY=257.6672, extentZ=96.96, center=<80 0 75.2>}]}
constraints []
